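# Season 33 story conversation in which the player asks Kendo about dark magic, covering Jade's disappearance and Daigo's experiments.
# One tagged entry point leads into a linear chain of four Speech nodes, advanced by Continue choices.
# The opening choice is categorized as main questline and requires the Maker story quest Q01 to be active.
# The final node's choice carries an Athena quest conversation tag, apparently for quest progress.
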ConversationDatabase /BattlePassS33_Quests/Conversations/StoryDialogue/Maker/Conversation_S33_MM_Q01
entry_points [{tag=Conversation.Entry.Season33.Story.MM.01, destinations [#1]}]
nodes [
 ConversationEntryPointNode_0 {
  id=#1
  class=ConversationEntryPointNode
  outputs [#2]
  properties {entryTag {tagName=Conversation.Entry.Season33.Story.MM.01}}
 }
 Speech_C_3 {
  id=#2
  class=Speech_C
  outputs [#3]
  properties {
   softTaskIcon=&T_UI_NPC_MainQuest
   generalConfig {defaultMessage="Jade has gone missing, and Daigo refuses to talk about it. I need to know why."}
  }
  sub_nodes [
   ConversationChoiceNode_5 {
    class=ConversationChoiceNode
    properties {defaultChoiceDisplayText=Ask Kendo about Dark Magic, choiceTags {gameplayTags [{tagName=Conversation.Choice.Category.MainQuestline}]}}
   }
   FortConversationRequirement_ControllerRequirement_4 {
    class=FortConversationRequirement_ControllerRequirement
    properties {
     nodeName=Require: Is Quest Active
     requirement {
      class=FortControllerRequirement_IsQuestActive
      properties {questDefinition=/BattlePassS33_Quests/Items/QuestItems/Story/S33_Story_Maker/quest_S33_story_maker_q01}
     }
    }
   }
  ]
 }
 Speech_C_28 {
  id=#3
  class=Speech_C
  outputs [#4]
  properties {generalConfig {defaultMessage=Jade and Daigo are both brilliant and reckless and... bad influences on one another.}}
  sub_nodes [ConversationChoiceNode_6 {class=ConversationChoiceNode, properties {defaultChoiceDisplayText=Continue}}]
 }
 Speech_C_0 {
  id=#4
  class=Speech_C
  outputs [#5]
  properties {generalConfig {defaultMessage="Before she vanished, Jade was investigating the influx of demons coming from the Spirit Realm."}}
  sub_nodes [ConversationChoiceNode_7 {class=ConversationChoiceNode, properties {defaultChoiceDisplayText=Continue}}]
 }
 Speech_C_1 {
  id=#5
  class=Speech_C
  outputs []
  properties {generalConfig {defaultMessage="Daigo has been experimenting with dark magic. If he's endangered Jade, I'll never forgive him."}}
  sub_nodes [ConversationChoiceNode_8 {
    class=ConversationChoiceNode
    properties {defaultChoiceDisplayText=Continue, choiceTags {gameplayTags [{tagName=Athena.Quests.S33.Conversation.Story.MM.01}]}}
   }]
 }
]
