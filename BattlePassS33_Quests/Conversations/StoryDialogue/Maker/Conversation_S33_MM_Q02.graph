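# Season 33 story conversation for the Maker questline's second quest, triggered when the player interacts with a portal.
# A single entry point leads to one speech node that shows the line about the shard as a "Look into the Portal" choice with an intel icon.
# The choice carries a portal-interact quest tag and requires the Maker Q02 quest to be active.
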ConversationDatabase /BattlePassS33_Quests/Conversations/StoryDialogue/Maker/Conversation_S33_MM_Q02
entry_points [{tag=Conversation.Entry.Season33.Story.MM.02.PortalInteract, destinations [#1]}]
nodes [
 ConversationEntryPointNode_0 {
  id=#1
  class=ConversationEntryPointNode
  outputs [#2]
  properties {entryTag {tagName=Conversation.Entry.Season33.Story.MM.02.PortalInteract}}
 }
 Speech_C_3 {
  id=#2
  class=Speech_C
  outputs []
  properties {softTaskIcon=&T_UI_NPC_Intel2, generalConfig {defaultMessage=The shard...the shard...}}
  sub_nodes [
   ConversationChoiceNode_5 {
    class=ConversationChoiceNode
    properties {
     defaultChoiceDisplayText=Look into the Portal
     choiceTags {gameplayTags [{tagName=Athena.Quests.S33.Conversation.Story.MM.02.PortalInteract}]}
    }
   }
   FortConversationRequirement_ControllerRequirement_4 {
    class=FortConversationRequirement_ControllerRequirement
    properties {
     nodeName=Require: Is Quest Active
     requirement {
      class=FortControllerRequirement_IsQuestActive
      properties {questDefinition=/BattlePassS33_Quests/Items/QuestItems/Story/S33_Story_Maker/quest_S33_story_maker_q02}
     }
    }
   }
  ]
 }
]
