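# Conversation database for a Season 33 Maker story quest dialogue, where the player is offered an option to investigate a painting.
# A single entry tag leads to one speech node pondering why Daigo gave Jade a magical mask, with one gameplay-tagged choice and no further outputs.
# An inverted quest-objective requirement limits that speech to players who have not completed the first objective of quest 03b.
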ConversationDatabase /BattlePassS33_Quests/Conversations/StoryDialogue/Maker/Conversation_S33_MM_Q03A
entry_points [{tag=Conversation.Entry.Season33.Story.MM.03.A, destinations [#1]}]
nodes [
 ConversationEntryPointNode_0 {
  id=#1
  class=ConversationEntryPointNode
  outputs [#2]
  properties {entryTag {tagName=Conversation.Entry.Season33.Story.MM.03.A}}
 }
 Speech_C_3 {
  id=#2
  class=Speech_C
  outputs []
  properties {
   softTaskIcon=&T_UI_NPC_question
   generalConfig {defaultMessage=Daigo gave Jade one of his magical masks. But why?}
  }
  sub_nodes [
   ConversationChoiceNode_5 {
    class=ConversationChoiceNode
    properties {defaultChoiceDisplayText=Investigate the Painting, choiceTags {gameplayTags [{tagName=Athena.Quests.S33.Conversation.Story.MM.03.A}]}}
   }
   FortConversationRequirement_ControllerRequirement_4 {
    class=FortConversationRequirement_ControllerRequirement
    properties {
     nodeName=Require: Has Completed Quest Objective
     requirement {
      class=FortControllerRequirement_HasCompletedQuestObjective
      properties {
       bInverseCondition=true
       objectiveBackendName=quest_S33_story_maker_q03b_obj0
       questDefinition=/BattlePassS33_Quests/Items/QuestItems/Story/S33_Story_Maker/quest_S33_story_maker_q03b
      }
     }
    }
   }
  ]
 }
]
