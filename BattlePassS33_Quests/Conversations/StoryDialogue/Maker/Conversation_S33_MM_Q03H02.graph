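# Season 33 story conversation in which the player confronts Daigo about Jade, and he explains the cursed mask he made to protect her.
# The entry tag leads into a linear chain of four speech nodes, each with a single choice.
# The opening choice shows a main-quest icon and requires the hidden Maker Q03 quest to be active.
# The last node's choice carries a quest gameplay tag, likely for progress tracking.
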ConversationDatabase /BattlePassS33_Quests/Conversations/StoryDialogue/Maker/Conversation_S33_MM_Q03H02
entry_points [{tag=Conversation.Entry.Season33.Story.MM.03.H02, destinations [#1]}]
nodes [
 ConversationEntryPointNode_0 {
  id=#1
  class=ConversationEntryPointNode
  outputs [#2]
  properties {entryTag {tagName=Conversation.Entry.Season33.Story.MM.03.H02}}
 }
 Speech_C_3 {
  id=#2
  class=Speech_C
  outputs [#3]
  properties {
   softTaskIcon=&T_UI_NPC_MainQuest
   generalConfig {defaultMessage=So you're the one digging into my business and messing with my brother's head.}
  }
  sub_nodes [
   ConversationChoiceNode_5 {class=ConversationChoiceNode, properties {defaultChoiceDisplayText=Confront Daigo about Jade}}
   FortConversationRequirement_ControllerRequirement_4 {
    class=FortConversationRequirement_ControllerRequirement
    properties {
     nodeName=Require: Is Quest Active
     requirement {
      class=FortControllerRequirement_IsQuestActive
      properties {questDefinition=/BattlePassS33_Quests/Items/QuestItems/Story/S33_Story_Maker/quest_S33_story_maker_q03_hidden_02}
     }
    }
   }
  ]
 }
 Speech_C_28 {
  id=#3
  class=Speech_C
  outputs [#4]
  properties {generalConfig {defaultMessage=I'd NEVER hurt Jade. She's the one person in the world I actually trust.}}
  sub_nodes [ConversationChoiceNode_6 {class=ConversationChoiceNode, properties {defaultChoiceDisplayText=Continue}}]
 }
 Speech_C_0 {
  id=#4
  class=Speech_C
  outputs [#5]
  properties {generalConfig {defaultMessage="The truth is, Jade got herself cursed by the demons, so I made a mask to protect her."}}
  sub_nodes [ConversationChoiceNode_7 {class=ConversationChoiceNode, properties {defaultChoiceDisplayText=Continue}}]
 }
 Speech_C_1 {
  id=#5
  class=Speech_C
  outputs []
  properties {generalConfig {defaultMessage="Yes, the mask made things worse, and I regret that... But I'm hoping it'll still keep her safe."}}
  sub_nodes [ConversationChoiceNode_8 {
    class=ConversationChoiceNode
    properties {defaultChoiceDisplayText=Continue, choiceTags {gameplayTags [{tagName=Athena.Quests.S33.Conversation.Story.MM.03.H02}]}}
   }]
 }
]
